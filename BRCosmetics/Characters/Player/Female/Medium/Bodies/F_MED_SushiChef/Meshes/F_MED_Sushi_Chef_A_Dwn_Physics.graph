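# Physics asset for the female medium Sushi Chef body mesh, setting up physics for its front apron and rear cloth.
# Two three-box chains (left and right) form the front apron and a two-box chain forms the rear cloth, all hung from the pelvis.
# Pelvis and thigh shapes supply the body volumes around the cloth.
# Swing is free on every joint, and an extra constraint ties the bottom left and right apron boxes together.
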PhysicsAsset /BRCosmetics/Characters/Player/Female/Medium/Bodies/F_MED_SushiChef/Meshes/F_MED_Sushi_Chef_A_Dwn_Physics
bodies {
 dyn_apron_frnt_rb_01_l [{shapeType=Box, extentX=1.5, extentY=11.5, extentZ=8.5, center=<3.75 -0.0002 0.0881>, rotation=<rot 0 -90 0>}]
 dyn_apron_frnt_rb_01_r [{shapeType=Box, extentX=1.5, extentY=11.5, extentZ=8.5, center=<-3.6445 0.0001 0>, rotation=<rot 0 -90 0>}]
 dyn_apron_frnt_rb_02_l [{shapeType=Box, extentX=1.75, extentY=9, extentZ=10, center=<3.4559 0.0001 0>, rotation=<rot 0 -90 0>}]
 dyn_apron_frnt_rb_02_r [{shapeType=Box, extentX=1.75, extentY=9, extentZ=10, center=<-3.4356 0 0>, rotation=<rot 0 -90 0>}]
 dyn_apron_frnt_rb_03_l [{shapeType=Box, extentX=1.95, extentY=12, extentZ=11, center=<3.9208 0.4245 -1.5668>, rotation=<rot 1.1091 -89.9769 6.362>}]
 dyn_apron_frnt_rb_03_r [{shapeType=Box, extentX=1.95, extentY=12, extentZ=11, center=<-3.8459 -0.3863 1.0804>, rotation=<rot 1.1572 -89.9733 6.6379>}]
 dyn_cloth_rear_mid_01 [{shapeType=Box, extentX=14.5, extentY=1, extentZ=10.5, center=<0 -0.2455 -5.5374>, rotation=<rot 0 0 5.0001>}]
 dyn_cloth_rear_mid_02 [{shapeType=Box, extentX=2.0214, extentY=11, extentZ=12.2519, center=<0.0054 0.3599 -3.7098>, rotation=<rot 5 90.0118 -89.7803>}]
 pelvis [
  {shapeType=Capsule, length=15.5, radius=9.75, center=<-6.9844 -2.1198 0>}
  {shapeType=Box, extentX=27.2654, extentY=36.7082, extentZ=10.1277, center=<1.8649 2.0002 -0.4472>, rotation=<rot 0 -12.6123 -84.8739>}
 ]
 thigh_l [{shapeType=Capsule, length=17.9922, radius=8.7, center=<-14.3756 0.8085 -0.447>, rotation=<rot 0.4884 -93.8431 -91.8795>}]
 thigh_r [{shapeType=Capsule, length=17.3798, radius=8.7, center=<14.0904 -0.6878 0.4172>, rotation=<rot 0.027 -95.5028 -93.601>}]
}
constraints [
 {
  bone1Name=dyn_apron_frnt_rb_02_l
  bone2Name=dyn_apron_frnt_rb_01_l
  swing1LimitDegrees=45
  swing1Motion=Free
  swing2LimitDegrees=45
  swing2Motion=Free
  twistLimitDegrees=45
 }
 {
  bone1Name=dyn_apron_frnt_rb_02_r
  bone2Name=dyn_apron_frnt_rb_01_r
  swing1LimitDegrees=45
  swing1Motion=Free
  swing2LimitDegrees=45
  swing2Motion=Free
  twistLimitDegrees=45
 }
 {
  bone1Name=dyn_apron_frnt_rb_03_r
  bone2Name=dyn_apron_frnt_rb_02_r
  swing1LimitDegrees=45
  swing1Motion=Free
  swing2LimitDegrees=45
  swing2Motion=Free
  twistLimitDegrees=60
 }
 {
  bone1Name=dyn_apron_frnt_rb_03_l
  bone2Name=dyn_apron_frnt_rb_02_l
  swing1LimitDegrees=45
  swing1Motion=Free
  swing2LimitDegrees=45
  swing2Motion=Free
  twistLimitDegrees=60
 }
 {
  bone1Name=thigh_l
  bone2Name=pelvis
  swing1LimitDegrees=45
  swing1Motion=Free
  swing2LimitDegrees=45
  swing2Motion=Free
  twistLimitDegrees=45
  twistMotion=Free
 }
 {
  bone1Name=thigh_r
  bone2Name=pelvis
  swing1LimitDegrees=45
  swing1Motion=Free
  swing2LimitDegrees=45
  swing2Motion=Free
  twistLimitDegrees=45
  twistMotion=Free
 }
 {
  bone1Name=dyn_apron_frnt_rb_01_l
  bone2Name=pelvis
  swing1LimitDegrees=45
  swing1Motion=Free
  swing2LimitDegrees=45
  swing2Motion=Free
  twistLimitDegrees=45
  twistMotion=Free
 }
 {
  bone1Name=dyn_apron_frnt_rb_01_r
  bone2Name=pelvis
  swing1LimitDegrees=45
  swing1Motion=Free
  swing2LimitDegrees=45
  swing2Motion=Free
  twistLimitDegrees=45
  twistMotion=Free
 }
 {
  bone1Name=dyn_apron_frnt_rb_03_l
  bone2Name=dyn_apron_frnt_rb_03_r
  swing1LimitDegrees=90
  swing1Motion=Free
  swing2LimitDegrees=90
  swing2Motion=Free
  twistLimitDegrees=90
  twistMotion=Free
 }
 {
  bone1Name=dyn_cloth_rear_mid_01
  bone2Name=pelvis
  swing1LimitDegrees=45
  swing1Motion=Free
  swing2LimitDegrees=45
  swing2Motion=Free
  twistLimitDegrees=10
  twistMotion=Free
 }
 {
  bone1Name=dyn_cloth_rear_mid_02
  bone2Name=dyn_cloth_rear_mid_01
  swing1LimitDegrees=45
  swing1Motion=Free
  swing2LimitDegrees=45
  swing2Motion=Free
  twistLimitDegrees=10
  twistMotion=Free
 }
]
