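# Physics asset for the bow on the female medium Sushi Chef player cosmetic, defining simulation bodies for its top left and right pieces.
# Small capsules on dyn_bow_top_l and dyn_bow_top_r each attach by a constraint to a larger spine_02 capsule.
# Both constraints share the same angular limits, with tight twist, narrow first swing and wider second swing, so each side moves mostly along one swing axis.
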PhysicsAsset /BRCosmetics/Characters/Player/Female/Medium/Bodies/F_MED_SushiChef/Meshes/F_MED_Sushi_Chef_Bow_Physics
bodies {
 dyn_bow_top_l [{shapeType=Capsule, length=1, radius=2.1245, center=<-0.01 0 0>, rotation=<rot -90 0 0>}]
 dyn_bow_top_r [{shapeType=Capsule, length=1, radius=2.0937, center=<0.01 0 0>, rotation=<rot -90 0 0>}]
 spine_02 [{shapeType=Capsule, length=4.8758, radius=2.8841, center=<2.2074 0.1071 0.0584>, rotation=<rot 0 5.2715 -173.6998>}]
}
constraints [
 {bone1Name=dyn_bow_top_l, bone2Name=spine_02, swing1LimitDegrees=10, swing2LimitDegrees=45, twistLimitDegrees=5}
 {bone1Name=dyn_bow_top_r, bone2Name=spine_02, swing1LimitDegrees=10, swing2LimitDegrees=45, twistLimitDegrees=5}
]
