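# Post-process style AnimBlueprint for the female Sushi Chef body cosmetic, adding secondary motion to the pose copied from the parent mesh.
# Pose Drivers and AnimDynamics correct elbows and sleeves, feeding a cached pose.
# With rigid bodies enabled, a bow sim follows crouch-switched physics assets, skipped when cloth is enabled; otherwise constraint and Flap nodes drive the apron from the thighs.
# The EventGraph overrides cloth gravity with skydiving force.
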AnimBlueprint /BRCosmetics/Characters/Player/Female/Medium/Bodies/F_MED_SushiChef/Meshes/F_MED_Sushi_Chef_AnimBP
target_skeleton=/BRCosmetics/Characters/Player/Female/Medium/Bodies/F_MED_SushiChef/Meshes/Fixed/F_MED_Sushi_Chef_Skeleton
graphs [
 AnimGraph {nodes [
   n1 = Break Fort Anim Input Skydiving External Force(FortAnimInput_SkydivingExternalForce=Get SkydivingExternalForceInput()) -> (
    bUseSkydivingVectorForce=~
    SkydivingVectorForce=~
    HeadToPelvisDirection=~
    FloatingMultiplier=~
    FloatingAdditive=~
    DivingMultiplier=~
    DivingAdditive=~
    ParachutingMultiplier=~
    ParachutingAdditive=~
   )
   n2 = RigidBody(ComponentPose=Local To Component(LocalPose=Use cached pose 'Cache1'()), ExternalForce=@n1.SkydivingVectorForce) {
    bClampLinearTranslationLimitToRefPose=true
    bOverrideWorldGravity=true
    overridePhysicsAsset=&/F_MED_Sushi_Chef_A_Physics
    simulationSpace=BaseBoneSpace
    alphaBoolBlend {blendOption=Linear}
    alphaScaleBiasClamp {bInterpResult=true}
    baseBoneRef {boneName=root}
    componentAppliedLinearAccClamp=<1000 1000 1100>
    componentLinearAccScale=<1 1 0.5>
    componentLinearVelScale=<1 1 0.5>
    overrideWorldGravity {z=-980}
   }
   n3 = RigidBody(ComponentPose=Local To Component(LocalPose=Use cached pose 'Cache1'()), ExternalForce=@n1.SkydivingVectorForce) {
    bClampLinearTranslationLimitToRefPose=true
    bOverrideWorldGravity=true
    overridePhysicsAsset=&/F_MED_Sushi_Chef_A_Dwn_Physics
    simulationSpace=BaseBoneSpace
    alphaBoolBlend {blendOption=Linear}
    alphaScaleBiasClamp {bInterpResult=true}
    baseBoneRef {boneName=pelvis}
    componentAppliedLinearAccClamp=<1100 1100 1100>
    componentLinearAccScale=<0.7 0.7 0.5>
    componentLinearVelScale=<0.7 0.7 0.5>
    overrideWorldGravity {z=-980}
   }
   n4 = Blend Poses by bool(
    BlendPose_0=Component To Local(ComponentPose=@n3)
    BlendPose_1=Component To Local(ComponentPose=@n2)
    BlendTime_0=0.15
    BlendTime_1=0.15
    bActiveValue=Get is_crouching() -> (is_crouching:bool)
   ) {blendType=Linear}
   n5 = Use cached pose 'Cache1'()
   n6 = Get bClothEnabled()
   n7 = RigidBody!(
    ComponentPose=Local To Component(LocalPose=Blend Poses by bool(BlendPose_0=@n5, BlendPose_1=@n4, bActiveValue=@n6) {blendType=Linear})
    Alpha
    ExternalForce
    ComponentLinearVelScale:Vector="0,0,0"
   ) -> (Pose) {
    overridePhysicsAsset=&/F_MED_Sushi_Chef_Bow_Physics
    simulationSpace=BaseBoneSpace
    alphaBoolBlend {blendOption=Linear}
    baseBoneRef {boneName=root}
    componentAppliedLinearAccClamp=<1000 1000 1000>
    componentLinearAccScale=<1 1 1>
    componentLinearVelScale=<50 50 50>
   }
   n8 = "Constraint - dyn_apron_frnt_rb_01_l (RR)"!(ComponentPose=Local To Component(LocalPose=Use cached pose 'Cache1'()), Alpha=1, ConstraintWeights_0=0.93, ConstraintWeights_1:float=0.07) -> (Pose) {
    alphaBoolBlend {blendOption=Linear}
    boneToModify {boneName=dyn_apron_frnt_rb_01_l}
    constraintSetup [{transformType=Rotation, targetBone {boneName=thigh_l}}, {transformType=Rotation, targetBone {boneName=pelvis}}]
    constraintWeights [1,1]
   }
   n9 = Flap Node - Bone: dyn_apron_frnt_rb_01_r!(ComponentPose:Component Space Pose Link Structure=@n8, Alpha:float=1) -> (Pose:Component Space Pose Link Structure) {
    is=title FortAnimGraphNode_Flap
    bUseClamp=true
    targetClampMax=0.1
    targetClampMin=-3.14
    targetComponent=RotationZ
    alphaBoolBlend {blendOption=Linear}
    sourceDrives [{
      bUseRange=false
      drivingCurve=null
      multiplier=-1
      rangeMax=1
      rangeMin=-1
      remappedMax=1
      remappedMin=0
      sourceComponent=RotationZ
      useQuaternion=true
      sourceBone {boneName=thigh_r}
     }]
    targetBone {boneName=dyn_apron_frnt_rb_01_r}
   }
   Output Pose(Result=Blend Poses by bool(BlendPose_0=Component To Local(ComponentPose=@n7), BlendPose_1=Component To Local(ComponentPose=@n9), bActiveValue=Get bRigidBodyEnabled()) {blendType=Linear}) {layerGroup=None}
   n10 = Pose Driver - Source: lowerarm_l!(SourcePose:Pose Link Structure=Copy Pose From Mesh() {bCopyCurves=true, bUseAttachedParent=true}) -> (Pose:Pose Link Structure) {
    is=title AnimGraphNode_PoseDriver
    poseAsset=&/F_MED_Sushi_Chef_Skeleton_PoseAsset
    onlyDriveBones [{boneName=driver_elbow_l}, {boneName=elbow_inside_l}, {boneName=elbow_inner_l}, {boneName=elbow_outer_l}, {boneName=driver_dyn_sleeve_l}]
    poseTargets [
     {
      bApplyCustomCurve=true
      bIsHidden=false
      distanceMethod=DefaultMethod
      drivenName=Pose_0
      functionType=DefaultFunction
      targetScale=1
      boneTransforms [{targetRotation=<rot 0 0 0>, targetTranslation=<0 0 0>}, {targetRotation=<rot 0 0 0>, targetTranslation=<0 0 0>}]
      customCurve {
       defaultValue=3.4028234663852886e+38
       postInfinityExtrap=RCCE_Constant
       preInfinityExtrap=RCCE_Constant
       keys [
        {
         arriveTangent=0
         arriveTangentWeight=0
         interpMode=RCIM_Linear
         leaveTangent=0
         leaveTangentWeight=0
         tangentMode=RCTM_Auto
         tangentWeightMode=RCTWM_WeightedNone
         time=0
         value=0
        }
        {
         arriveTangent=0
         arriveTangentWeight=0
         interpMode=RCIM_Linear
         leaveTangent=0
         leaveTangentWeight=0
         tangentMode=RCTM_Auto
         tangentWeightMode=RCTWM_WeightedNone
         time=1
         value=1
        }
       ]
      }
      targetRotation=<rot 0 0 0>
     }
     {
      bApplyCustomCurve=false
      bIsHidden=false
      distanceMethod=DefaultMethod
      drivenName=Pose_1
      functionType=DefaultFunction
      targetScale=1
      boneTransforms [{targetRotation=<rot 0 0 0>, targetTranslation=<0 0 0>}, {targetRotation=<rot 0 0 0>, targetTranslation=<0 0 0>}]
      customCurve {defaultValue=3.4028234663852886e+38, postInfinityExtrap=RCCE_Constant, preInfinityExtrap=RCCE_Constant, keys []}
      targetRotation=<rot 0 0 0>
     }
     {
      bApplyCustomCurve=false
      bIsHidden=false
      distanceMethod=DefaultMethod
      drivenName=Pose_2
      functionType=DefaultFunction
      targetScale=1
      boneTransforms [{targetRotation=<rot 0 0 0>, targetTranslation=<0 0 0>}, {targetRotation=<rot 0 0 0>, targetTranslation=<0 0 0>}]
      customCurve {defaultValue=3.4028234663852886e+38, postInfinityExtrap=RCCE_Constant, preInfinityExtrap=RCCE_Constant, keys []}
      targetRotation=<rot 0 0 0>
     }
     {
      bApplyCustomCurve=false
      bIsHidden=false
      distanceMethod=DefaultMethod
      drivenName=Pose_3
      functionType=DefaultFunction
      targetScale=1
      boneTransforms [{targetRotation=<rot 0 0 0>, targetTranslation=<0 0 0>}, {targetRotation=<rot 0 0 0>, targetTranslation=<0 0 0>}]
      customCurve {defaultValue=3.4028234663852886e+38, postInfinityExtrap=RCCE_Constant, preInfinityExtrap=RCCE_Constant, keys []}
      targetRotation=<rot 0 0 0>
     }
     {
      bApplyCustomCurve=false
      bIsHidden=false
      distanceMethod=DefaultMethod
      drivenName=Pose_4
      functionType=DefaultFunction
      targetScale=1
      boneTransforms [{targetRotation=<rot 0 0 0>, targetTranslation=<0 0 0>}, {targetRotation=<rot 0 0 0>, targetTranslation=<0 0 0>}]
      customCurve {defaultValue=3.4028234663852886e+38, postInfinityExtrap=RCCE_Constant, preInfinityExtrap=RCCE_Constant, keys []}
      targetRotation=<rot 0 0 0>
     }
     {
      bApplyCustomCurve=false
      bIsHidden=false
      distanceMethod=DefaultMethod
      drivenName=Pose_5
      functionType=DefaultFunction
      targetScale=1
      boneTransforms [{targetRotation=<rot 0 0 0>, targetTranslation=<0 0 0>}, {targetRotation=<rot 0 0 0>, targetTranslation=<0 0 0>}]
      customCurve {defaultValue=3.4028234663852886e+38, postInfinityExtrap=RCCE_Constant, preInfinityExtrap=RCCE_Constant, keys []}
      targetRotation=<rot 0 0 0>
     }
     {
      bApplyCustomCurve=false
      bIsHidden=false
      distanceMethod=DefaultMethod
      drivenName=Pose_6
      functionType=DefaultFunction
      targetScale=1
      boneTransforms [{targetRotation=<rot 0 0 0>, targetTranslation=<0 0 0>}, {targetRotation=<rot 0 0 0>, targetTranslation=<0 0 0>}]
      customCurve {defaultValue=3.4028234663852886e+38, postInfinityExtrap=RCCE_Constant, preInfinityExtrap=RCCE_Constant, keys []}
      targetRotation=<rot 0 0 0>
     }
     {
      bApplyCustomCurve=false
      bIsHidden=false
      distanceMethod=DefaultMethod
      drivenName=Pose_7
      functionType=DefaultFunction
      targetScale=1
      boneTransforms [{targetRotation=<rot 0 0 0>, targetTranslation=<0 0 0>}, {targetRotation=<rot 0 0 0>, targetTranslation=<0 0 0>}]
      customCurve {defaultValue=3.4028234663852886e+38, postInfinityExtrap=RCCE_Constant, preInfinityExtrap=RCCE_Constant, keys []}
      targetRotation=<rot 0 0 0>
     }
    ]
    rBFParams {distanceMethod=Quaternion, radius=55}
    sourceBones [{boneName=lowerarm_l}]
   }
   n11 = Pose Driver - Source: lowerarm_r!(SourcePose:Pose Link Structure=@n10) -> (Pose:Pose Link Structure) {
    is=title AnimGraphNode_PoseDriver
    poseAsset=&F_MED_Sushi_Chef_Skeleton_PoseAsset
    onlyDriveBones [{boneName=driver_elbow_r}, {boneName=elbow_inside_r}, {boneName=elbow_inner_r}, {boneName=elbow_outer_r}, {boneName=driver_dyn_sleeve_r}]
    poseTargets [
     {
      bApplyCustomCurve=false
      bIsHidden=false
      distanceMethod=DefaultMethod
      drivenName=Pose_0
      functionType=DefaultFunction
      targetScale=1
      boneTransforms [{targetRotation=<rot 1.7008 26.5846 0.8306>, targetTranslation=<-26.4651 0.8345 -0.259>}]
      customCurve {defaultValue=3.4028234663852886e+38, postInfinityExtrap=RCCE_Constant, preInfinityExtrap=RCCE_Constant, keys []}
      targetRotation=<rot 0 0 0>
     }
     {
      bApplyCustomCurve=false
      bIsHidden=false
      distanceMethod=DefaultMethod
      drivenName=Pose_1
      functionType=DefaultFunction
      targetScale=0.1361
      boneTransforms [{targetRotation=<rot -0.2203 114.6092 -0.6612>, targetTranslation=<-26.4651 0.8345 -0.259>}]
      customCurve {defaultValue=3.4028234663852886e+38, postInfinityExtrap=RCCE_Constant, preInfinityExtrap=RCCE_Constant, keys []}
      targetRotation=<rot 0 0 0>
     }
     {
      bApplyCustomCurve=false
      bIsHidden=false
      distanceMethod=DefaultMethod
      drivenName=Pose_2
      functionType=DefaultFunction
      targetScale=0.1387
      boneTransforms [{targetRotation=<rot -0.0782 107.8265 -0.5542>, targetTranslation=<-26.4651 0.8345 -0.259>}]
      customCurve {defaultValue=3.4028234663852886e+38, postInfinityExtrap=RCCE_Constant, preInfinityExtrap=RCCE_Constant, keys []}
      targetRotation=<rot 0 0 0>
     }
     {
      bApplyCustomCurve=false
      bIsHidden=false
      distanceMethod=DefaultMethod
      drivenName=Pose_3
      functionType=DefaultFunction
      targetScale=0.0329
      boneTransforms [{targetRotation=<rot -0.1853 110.7678 -0.5612>, targetTranslation=<-26.4651 0.8345 -0.259>}]
      customCurve {defaultValue=3.4028234663852886e+38, postInfinityExtrap=RCCE_Constant, preInfinityExtrap=RCCE_Constant, keys []}
      targetRotation=<rot 0 0 0>
     }
     {
      bApplyCustomCurve=false
      bIsHidden=false
      distanceMethod=DefaultMethod
      drivenName=Pose_4
      functionType=DefaultFunction
      targetScale=1
      boneTransforms [{targetRotation=<rot 0.6608 47.7825 0.7637>, targetTranslation=<-26.4651 0.8345 -0.259>}]
      customCurve {defaultValue=3.4028234663852886e+38, postInfinityExtrap=RCCE_Constant, preInfinityExtrap=RCCE_Constant, keys []}
      targetRotation=<rot 0 0 0>
     }
     {
      bApplyCustomCurve=false
      bIsHidden=false
      distanceMethod=DefaultMethod
      drivenName=Pose_5
      functionType=DefaultFunction
      targetScale=0.3105
      boneTransforms [{targetRotation=<rot -0.282 121.1986 -0.8364>, targetTranslation=<-26.4651 0.8345 -0.259>}]
      customCurve {defaultValue=3.4028234663852886e+38, postInfinityExtrap=RCCE_Constant, preInfinityExtrap=RCCE_Constant, keys []}
      targetRotation=<rot 0 0 0>
     }
     {
      bApplyCustomCurve=false
      bIsHidden=false
      distanceMethod=DefaultMethod
      drivenName=Pose_6
      functionType=DefaultFunction
      targetScale=0.0314
      boneTransforms [{targetRotation=<rot 0.0995 111.7848 -0.1452>, targetTranslation=<-26.4651 0.8345 -0.259>}]
      customCurve {defaultValue=3.4028234663852886e+38, postInfinityExtrap=RCCE_Constant, preInfinityExtrap=RCCE_Constant, keys []}
      targetRotation=<rot 0 0 0>
     }
     {
      bApplyCustomCurve=false
      bIsHidden=false
      distanceMethod=DefaultMethod
      drivenName=Pose_7
      functionType=DefaultFunction
      targetScale=0.0314
      boneTransforms [{targetRotation=<rot 0.0561 111.132 -0.0175>, targetTranslation=<-26.4651 0.8345 -0.259>}]
      customCurve {defaultValue=3.4028234663852886e+38, postInfinityExtrap=RCCE_Constant, preInfinityExtrap=RCCE_Constant, keys []}
      targetRotation=<rot 0 0 0>
     }
    ]
    rBFParams {distanceMethod=Quaternion, radius=55}
    sourceBones [{boneName=lowerarm_r}]
   }
   n12 = Anim Dynamics - Bone: elbow_outer_l(ComponentPose=Local To Component(LocalPose=@n11), Alpha=0.55, ExternalForce="0,0,0") {
    bGravityOverrideInSimSpace=true
    gravityScale=0.3
    linearDampingOverride=0.9999
    linearSpringConstant=100
    alphaBoolBlend {blendOption=Linear}
    boundBone {boneName=elbow_outer_l}
    physicsBodyDefinitions [{
      sphereCollisionRadius=0
      boundBone {boneName=elbow_outer_l}
      boxExtents=<20 20 20>
      constraintSetup {angularLimitsMax=<0 0 0>, angularLimitsMin=<0 0 0>, angularTarget=<0 0 0>, linearAxesMax=<0.5 2 1>, linearAxesMin=<-0.5 -2 -1>}
      localJointOffset=<-7 0 0>
     }]
   }
   n13 = Anim Dynamics - Bone: elbow_outer_r(ComponentPose=@n12, Alpha=0.55, ExternalForce="0,0,0") {
    bGravityOverrideInSimSpace=true
    gravityScale=0.3
    linearDampingOverride=0.9999
    linearSpringConstant=100
    alphaBoolBlend {blendOption=Linear}
    boundBone {boneName=elbow_outer_r}
    physicsBodyDefinitions [{
      sphereCollisionRadius=0
      boundBone {boneName=elbow_outer_r}
      boxExtents=<20 20 20>
      constraintSetup {angularLimitsMax=<0 0 0>, angularLimitsMin=<0 0 0>, angularTarget=<0 0 0>, linearAxesMax=<0.5 2 1>, linearAxesMin=<-0.5 -2 -1>}
      localJointOffset=<7 0 0>
     }]
   }
   Save cached pose 'Cache1'(Pose=Component To Local(ComponentPose=@n13)) {cacheName=Cache1}
  ]}
 EventGraph {nodes [
   Event BlueprintUpdateAnimation()
   > n1 = Branch(Condition=Get is_skydiving() -> (is_skydiving:bool))
   n2 = Get clothing_sim_interactor() -> (clothing_sim_interactor:Clothing Simulation Interactor Nv Object Reference)
   > n3 = EnableGravityOverride!(
    execute
    self:Clothing Simulation Interactor Object Reference=@n2
    InVector:"Vector (by ref)"=Break Fort Anim Input Skydiving External Force(FortAnimInput_SkydivingExternalForce=Get SkydivingExternalForceInput()) -> (
     bUseSkydivingVectorForce=~
     SkydivingVectorForce=~
     HeadToPelvisDirection=~
     FloatingMultiplier=~
     FloatingAdditive=~
     DivingMultiplier=~
     DivingAdditive=~
     ParachutingMultiplier=~
     ParachutingAdditive=~
    ).SkydivingVectorForce
   ) -> (then)
   ForceClothNextUpdateTeleportAndReset(
    execute=[@n3.then, @n4.Completed]
    self:Skeletal Mesh Component Object Reference=Get MeshToCopy() -> (MeshToCopy:Skeletal Mesh Component Object Reference)
   )
   DisableGravityOverride(execute=@n1.else, self:Clothing Simulation Interactor Object Reference=@n2)
   > n4 = Do Once()
   Event BlueprintInitializeAnimation()
   > n5 = Cast To ClothingSimulationInteractorNv!(
    execute
    Object=GetClothingSimulationInteractor!(self:Skeletal Mesh Component Object Reference=GetPartSkeletalMeshComponent(), ClothingAssetName:Name=None) -> (ReturnValue:Clothing Simulation Interactor Object Reference)
   ) -> (then, CastFailed, AsClothing Simulation Interactor Nv:Clothing Simulation Interactor Nv Object Reference) {is=title K2Node_DynamicCast}
   > Set clothing_sim_interactor(clothing_sim_interactor:Clothing Simulation Interactor Nv Object Reference=@n5) -> (Output_Get:Clothing Simulation Interactor Nv Object Reference)
   > Set MeshToCopy(MeshToCopy:Skeletal Mesh Component Object Reference=GetOwnerSkeletalMeshComponent()) -> (Output_Get:Skeletal Mesh Component Object Reference)
   Event OnBeginSkydiving()
   > Set is_skydiving(is_skydiving:bool=true) -> (Output_Get:bool)
   Event OnEndSkydiving()
   > Set is_skydiving(is_skydiving=false)
   Event OnBeginCrouching()
   > Set is_crouching(is_crouching:bool=true) -> (Output_Get:bool)
   Event OnEndCrouching()
   > Set is_crouching(is_crouching=false)
   n6 = Branch(Condition=Get is_skydiving())
   n7 = Branch(execute=[@n6.else, @n6.then], Condition=Get bIsSkydiveFloating())
   n8 = Get Direction Between Sockets!(
    Component:Skeletal Mesh Component Object Reference=GetOwnerSkeletalMeshComponent()
    SocketOrBoneNameFrom:Name=head
    SocketOrBoneNameTo:Name=pelvis
   ) -> (ReturnValue:Vector) {is=title K2Node_CallFunction}
   > Set skydiving vector force(skydiving vector force:Vector=vector + vector(
     A:Vector=vector * vector(A:Vector=@n8, B:Vector="75,75,75") -> (ReturnValue:Vector) {is=title K2Node_CommutativeAssociativeBinaryOperator}
     B:Vector="0, 0, 0"
    ) -> (ReturnValue:Vector) {is=title K2Node_CommutativeAssociativeBinaryOperator}) -> (Output_Get:Vector)
   n9 = Branch(execute=@n7.else, Condition=Get bIsSkydiveDiveMode())
   > Set skydiving vector force(skydiving vector force=vector + vector(A=vector * vector(A=@n8, B="180,180,200"), B="0, 0,1000"))
   Set skydiving vector force(execute=@n9.else, skydiving vector force=vector * vector(A=@n8, B="75,75,75"))
   Reroute Node()
   n10 = GetOwnerSkeletalMeshComponent()
   n11 = GetDeltaTransformFromRefPose(self=@n10, BoneName=pelvis, BaseName=thigh_l)
   n12 = MapRangeClamped(Value=BreakRotator(InRot=BreakTransform(InTransform=@n11).Rotation).Yaw, InRangeA=10, InRangeB=120)
   > Set thigh_float_l(thigh_float_l:double=@n12) -> (Output_Get:double)
   n13 = GetDeltaTransformFromRefPose(self=@n10, BoneName=pelvis, BaseName=thigh_r)
   n14 = MapRangeClamped(Value=BreakRotator(InRot=BreakTransform(InTransform=@n13).Rotation).Yaw, InRangeA=10, InRangeB=120)
   > Set thigh_float_r(thigh_float_r:double=@n14) -> (Output_Get:double)
   Reroute Node(InputPin:Vector=vector * vector(A="0, 0, 0", B="-5,-5,5")) -> (OutputPin:Vector)
   Set thigh_l_rotate(thigh_l_rotate:Rotator=MakeRotator(Roll=@n12)) -> (Output_Get:Rotator)
   Set thigh_r_rotate(thigh_r_rotate:Rotator=MakeRotator(Roll=@n14)) -> (Output_Get:Rotator)
  ]}
]
variables [
 clothing_sim_interactor {type=ClothingSimulationInteractorNv, default=null}
 is_crouching {type=boolean}
 is_skydiving {type=boolean}
 MeshToCopy {type=SkeletalMeshComponent}
 skydiving vector force {type=Vector}
 thigh_float_l {type=number}
 thigh_float_r {type=number}
 thigh_l_rotate {type=Rotator}
 thigh_r_rotate {type=Rotator}
]
